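# Translucent UI material that draws the PoppySoap reticle, with parameter-driven states for enemy in range, MotoWrestle and HookShot scale.
# MSDF textures form outer brackets, inner spikes and a mirrored, radially masked spike, while an SDF circle outline forms the center.
# Anti-aliased thresholds build shapes and softer outlines, combined through named reroutes into color and opacity.
# MotoWrestle flips ring UVs and changes spike tint; Enemy In Range shrinks the circle and changes color.
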Material /PoppySoap/Assets/UI/M_UI_Ret_PoppySoap_V2
graph {
 properties {bAutomaticallySetUsageInEditor=false, blendMode=BLEND_Translucent, materialDomain=MD_UI, usedShadingModels=Default Lit}
 nodes [
  LinearInterpolate(Alpha=ScalarParameter() {parameterName=MotoWrestle Switch}) {constA=6.5, constAlpha=1, constB=0}
  n1 = Constant3Vector() {constant=<rgba 0 0 0 1>}
  n2 = LinearInterpolate(
   A=Constant3Vector() {constant=<rgba 0.3151 0.3151 0.3151 1>}
   B=LinearInterpolate(
    A=Constant3Vector() {constant=<rgba 0.5052 0.3395 0 1>}
    B=Constant3Vector() {constant=<rgba 0.1999 0.7792 0.9047 1>}
    Alpha=ScalarParameter() {parameterName=MotoWrestle Switch}
   )
   Alpha=ScalarParameter() {parameterName=Spikes Color Switch}
  )
  n3 = NamedRerouteUsage()
  LinearInterpolate(
   A=LinearInterpolate(A=LinearInterpolate(A=@n1, B=@n2, Alpha=@n3), Alpha=NamedRerouteUsage()) {constA=1}
   B=LinearInterpolate(
    A=Constant3Vector() {constant=<rgba 0.9047 0.9047 0.9047 1>}
    B=Constant3Vector() {constant=<rgba 0.9047 0.5508 0 1>}
    Alpha=ScalarParameter() {parameterName=Enemy In Range}
   )
   Alpha=NamedRerouteUsage()
  )
  MaterialFunctionCall() {materialFunction=&DebugTimeSine}
  n4 = ScalarParameter() {parameterName=Reticle Master Opacity}
  Max(A=Multiply(A=NamedRerouteUsage(), B=LinearInterpolate(Alpha=@n4) {constA=0.5, constB=0.85}), B=NamedRerouteUsage())
  n5 = MaterialFunctionCall() {materialFunction=&GetUserInterfaceUV}
  n6 = OneMinus(None=ComponentMask(None=@n5.9-Slice UV) {r=true})
  n7 = ComponentMask(None=@n5.9-Slice UV) {g=true}
  n8 = MaterialFunctionCall(
   UVs=LinearInterpolate(A=@n5.9-Slice UV, B=AppendVector(A=@n6, B=@n7), Alpha=ScalarParameter() {parameterName=MotoWrestle Switch}) {constA=6.5, constAlpha=1, constB=0}
   Texture Scale=LinearInterpolate(Alpha=ScalarParameter() {parameterName=HookShot Scale}) {constA=0.8}
  ) {materialFunction=&ScaleUVsByCenter}
  n9 = Constant() {r=6.5}
  n10 = MaterialFunctionCall!(MSDF RGB=TextureSample(UVs=Rotator(Coordinate=@n8.UVs, Time=@n9)) {samplerType=SAMPLERTYPE_LinearColor, texture=&/T_UI_MSDF_PoppySoap_OuterRing}.RGB) {materialFunction=/FortUILibrary/MaterialFunctions/Shapes/Modifiers/MF_UI_MSDFtoSDF}
  n11 = Constant() {r=0.5}
  n12 = Constant() {r=0.06}
  n13 = MaterialFunctionCall(DistanceField=@n10.SDF, Threshold=@n11, Smoothing=@n12) {materialFunction=&MF_UI_AntiAliasing}
  n14 = ScalarParameter() {defaultValue=0.9, parameterName=Outer Bracket Opacity}
  n15 = NamedRerouteDeclaration(None=Max(A=Multiply(A=@n13.Outer, B=@n14) {constB=0}, B=NamedRerouteUsage())) {name=Outer Brackets, nodeColor=<rgba 0 0 0 1>}
  n16 = MaterialFunctionCall(UVs=@n8.UVs, Texture Scale=Constant() {r=1.047}) {materialFunction=&ScaleUVsByCenter}
  n17 = Constant() {r=6.3192}
  n18 = MaterialFunctionCall(MSDF RGB=TextureSample(UVs=Rotator(Coordinate=@n16.UVs, Time=@n17)) {samplerType=SAMPLERTYPE_LinearColor, texture=&/T_UI_MSDF_PoppySoap_InnerRing}.RGB) {materialFunction=&MF_UI_MSDFtoSDF}
  n19 = Multiply(
   A=MaterialFunctionCall(DistanceField=@n18.SDF, Threshold=@n11, Smoothing=@n12) {materialFunction=&MF_UI_AntiAliasing}.Outer
   B=ScalarParameter() {defaultValue=0.8, group=Opacity, parameterName=Inner Spikes Opacity}
  )
  n20 = NamedRerouteDeclaration(None=Saturate(None=Subtract(A=@n19, B=@n13.Outer))) {name=Inner Spikes, nodeColor=<rgba 0 0 0 1>}
  n21 = Constant() {r=0.38}
  n22 = Max(A=@n10.SDF, B=@n18.SDF)
  NamedRerouteDeclaration(None=LinearInterpolate(
    A=Max(A=@n15, B=@n20)
    B=MaterialFunctionCall(DistanceField=@n22, Threshold=@n21, Smoothing=@n12) {materialFunction=&MF_UI_AntiAliasing}.Outer
   ) {constAlpha=0.4}) {name=Reticle Outline, nodeColor=<rgba 0 0 0 1>}
  n23 = Constant() {r=0.5}
  n24 = Constant() {r=0.05}
  n25 = LinearInterpolate(A=ScalarParameter() {defaultValue=0.1, parameterName=CircleSize}, Alpha=ScalarParameter() {parameterName=Enemy In Range}) {constB=0.01}
  n26 = Multiply(A=MaterialFunctionCall!(
    DistanceField=MaterialFunctionCall!(UVs, Radius=@n25) {materialFunction=/FortUILibrary/MaterialFunctions/Shapes/SDFs/MF_SDF_Circle}.SDF
    Threshold
    Thickness=ScalarParameter() {defaultValue=0.065, parameterName=CircleThickness}
    Alignment
    Smoothing
   ) {materialFunction=/FortUILibrary/MaterialFunctions/Shapes/Modifiers/MF_UI_ShapeOutline}.SDFOutline) {constB=-10}
  n27 = MaterialFunctionCall(DistanceField=@n26, Threshold=@n23, Smoothing=@n24) {materialFunction=&MF_UI_AntiAliasing}
  n28 = MaterialFunctionCall(DistanceField=@n26, Threshold=Add(A=@n23) {constB=-0.15}, Smoothing=@n24) {materialFunction=&MF_UI_AntiAliasing}
  NamedRerouteDeclaration(None=LinearInterpolate(A=NamedRerouteDeclaration(None=@n27.Outer) {name=Center, nodeColor=<rgba 0 0 0 1>}, B=Multiply(A=@n28.Outer) {constB=0.7}) {constAlpha=0.4}) {name=Center Outline, nodeColor=<rgba 0 0 0 1>}
  n29 = Rotator(Time=LinearInterpolate(A=Constant() {r=15.54}, B=Constant() {r=-15.5}, Alpha=ScalarParameter() {parameterName=MotoWrestle Switch}) {constA=-15.5, constAlpha=1, constB=15.53})
  n30 = ScalarParameter() {parameterName=In Range Spike Offset, sliderMax=1}
  n31 = Constant2Vector() {g=0.01, r=-0.07}
  n32 = AppendVector(A=Add(A=LinearInterpolate(Alpha=@n30) {constA=0.17, constB=0.361}, B=@n31.R), B=@n31.G)
  n33 = Multiply(A=Constant2Vector() {g=0.5, r=1}) {constB=0.5192}
  n34 = TextureSample(UVs=Rotator(
    Coordinate=MaterialFunctionCall!(UV=@n29, Offset=@n32, Rotation, Scale=@n33, MirrorY) {materialFunction=/FortUILibrary/MaterialFunctions/Basics/MF_UI_Symmetry_2x1_DEPRECATED}.Result
    Time=Constant() {r=-18.9}
   )) {samplerSource=SSM_Clamp_WorldGroupSettings, samplerType=SAMPLERTYPE_LinearColor, texture=&/T_UI_MSDF_PoppySoap_Spike}
  n35 = Constant() {r=0.284}
  n36 = Multiply(
   A=MaterialFunctionCall(
    DistanceField=MaterialFunctionCall(MSDF RGB=@n34.RGB) {materialFunction=&MF_UI_MSDFtoSDF}.SDF
    Threshold=Constant() {r=0.4}
   ) {materialFunction=&MF_UI_AntiAliasing}.Outer
   B=MaterialFunctionCall(GradientStart=@n35, GradientEnd=Subtract(A=@n35) {constB=0.02}) {materialFunction=&MF_UI_RadialGradient}.Smooth
  )
  n37 = ScalarParameter() {defaultValue=0.9, parameterName=Outer Bracket Opacity}
  n38 = ScalarParameter() {parameterName=Outer Bracket Spike Opacity Switch}
  NamedRerouteDeclaration(None=Multiply(A=Multiply(A=@n36, B=@n37) {constB=0.8}, B=LinearInterpolate(Alpha=@n38))) {name=Spikey Thing, nodeColor=<rgba 0 0 0 1>}
 ]
}
